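; Gameplay ability for the ChapWorm super sprint, adding a camera mode and modifier on activation and ending the kinetic-energy charge tutorial if active.
; After a hotfixable delay, it applies the threshold effect once the player is walking, retrying on movement-mode changes, if sprint energy suffices.
; An async tag wait grants fall-damage immunity upon sprint jumping.
; Ending clears timer, threshold effect and bindings, removing camera changes through a cached manager that survives pawn destruction.
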(event Ability|EventActivateAbility
  (bind _returnvalue (CallFunction|GetAvatarasFortPlayerPawn))
  ;[0.005189 0.08333 0.005165] ActivateAbility: Start a timer after which we apply the threshold gameplay and visuals.
  (|CallParentFunction)
  (bind _returnvalue_2 (Utilities|Time|SetTimerbyEvent (EventDispatchers|CreateEvent :!event Custom|SuperSprintAccelerationComplete) (ScalableFloat|GetValueatLevel (Variables|Hotfix|GetHotfix_SuperSprintThresholdActivateDelay))))
  (Variables|Transient|SetSuperSprintAccelerationCompleteTimerHandle _returnvalue_2)
  (Default|AddCameraMode)
  (bind _outfortplayerpawn (CallFunction|GetAvatarasValidFortPlayerPawn)
    (:ValidFortPlayerPawn
      (bind _asynctaskproxy (Ability|Tasks|WaitGameplayTagAddToActor _outfortplayerpawn "(TagName=\"Gameplay.Action.Player.SprintJumping\")")
        (:then
          (Variables|Transient|SetAsyncAction_SprintJumping _asynctaskproxy)
          ;; End tutorial for kinetic energy charge.
          (if (Actor|ActorHasTag _returnvalue "ChapWormTutorialKineticEnergyChargeActive")
            (Ability|SendGameplayEventtoActor _returnvalue "(TagName=\"ChapWorm.Tutorial.KineticEnergy.Charge.End\")")))
        (:Added
          (Ability|ApplyGameplayEffectToOwner (Variables|Settings|GetFallDamageImmunityGameplayEffectClass) 0))))))

(event Custom|SuperSprintAccelerationComplete
  ;[0.1458 0.03823 0.1449] SuperSprintAccelerationComplete: Acceleration has been completed, apply the threshold gameplay effect.
  ;;[0.01842 0.09074 0.1094] Wait for the player to return to 'Walking' movement mode in case they are not in it.
  (bind _outfortplayerpawn (CallFunction|GetAvatarasValidFortPlayerPawn)
    (:ValidFortPlayerPawn
      (bind _charactermovement (Variables|/Script/Engine.Character|GetCharacterMovement _outfortplayerpawn)
        (:then
          (if (Pawn|Components|CharacterMovement|IsWalking _charactermovement)
            (Character|UnbindEventfromMovementModeChangedDelegate _outfortplayerpawn (delegate Custom|HandleMovementModeChanged))
            ;;[0.02472 0.1146 0.02518] Sufficient energy is checked before the threshold values are played, otherwise it feels bad when the effect plays and the player immediately stops sprinting cause they are out of energy.
            (bind _energycomponent (Default|GetTacticalSprintEnergyComponentBranching)
              (:Is%20Valid
                (if (Energy|HasSufficientEnergy _energycomponent (ScalableFloat|GetValueatLevel (Variables|Hotfix|GetHotfix_SuperSprintSufficientEnergyThreshold)))
                  ;;[0.3193 0.2779 1] Apply the threshold gameplay effect on the player.
                  (bind _returnvalue (Ability|ApplyGameplayEffectToOwner "/ChapWormGameplay/Items/ChapWorm/Abilities/SuperSprint/GameplayEffects/GE_ChapWorm_SuperSprintThreshold.GE_ChapWorm_SuperSprintThreshold_C"))
                  (Variables|Transient|SetSuperSprintThresholdGameplayEffectHandle _returnvalue)
                  (else
                    (Fortnite|FortShippingLog self "SuperSprintAccelerationComplete: Not enough stamina to activate SuperSprintThreshold.")))))
            (else
              (Character|BindEventtoMovementModeChangedDelegate _outfortplayerpawn (delegate Custom|HandleMovementModeChanged)))))
        (:else)))))

(event Custom|HandleMovementModeChanged
  (params (Character :type Character%20Object%20Reference) (PrevMovementMode :type EMovementMode%20Enum :default "MOVE_None") (PreviousCustomMode :type Byte))
  ;[0.1458 0.03823 0.1449] SuperSprintAccelerationComplete: Acceleration has been completed, apply the threshold gameplay effect.
  ;;[0.01842 0.09074 0.1094] Wait for the player to return to 'Walking' movement mode in case they are not in it.
  (CallFunction|SuperSprintAccelerationComplete))

(event Ability|EventOnEndAbility
  (params (bWasCancelled :type Boolean))
  ;[0.07812 0.0487 0.009052] EndAbility: Clean up.
  (Utilities|FlowControl|Sequence
    (:then_0
      (Ability|RemoveGameplayEffectFromOwnerWithHandle (Variables|Transient|GetSuperSprintThresholdGameplayEffectHandle))
      (Utilities|Time|ClearandInvalidateTimerbyHandle (Variables|Transient|GetSuperSprintAccelerationCompleteTimerHandle))
      (Default|RemoveCameraMode)
      (bind _asyncaction_sprintjumping (Variables|Transient|GetAsyncAction_SprintJumping)
        (:then
          (Ability|Async|EndAction _asyncaction_sprintjumping))
        (:else
          _))
      (bind _outfortplayerpawn (CallFunction|GetAvatarasValidFortPlayerPawn)
        (:ValidFortPlayerPawn
          (Character|UnbindEventfromMovementModeChangedDelegate _outfortplayerpawn (EventDispatchers|CreateEvent :!event Custom|HandleMovementModeChanged)))))
    (:then_1
      (|CallParentFunction bWasCancelled))))

(fn AddCameraMode
  ;[0.000237 0.04228 0.05729] AddCameraMode: Add the desired camera mode to the player pawn.
  (bind _outfortplayerpawn (CallFunction|GetAvatarasValidFortPlayerPawn)
    (:ValidFortPlayerPawn
      (Utilities|IsValid (Fortnite|GetFortPlayerControllerfromActor _outfortplayerpawn)
        (:Is%20Valid
          (bind _playercameramanager (Variables|/Script/Engine.PlayerController|GetPlayerCameraManager (Fortnite|GetFortPlayerControllerfromActor _outfortplayerpawn))
            (:then
              (bind _asfort_player_camera_base (Utilities|Casting|CastToFortPlayerCameraBase _playercameramanager))
              (bind _output_get (Variables|Transient|SetCachedCameraManager _asfort_player_camera_base)) ; Cache in case pawn is destroyed\nbefore we try to remove modifier
              (Camera|AddCameraModeOverride _output_get (Variables|Camera|GetCameraModeClass) (Variables|Camera|GetCameraModeTag) 0 0 0)
              (bind _returnvalue (Game|Player|AddNewCameraModifier _output_get "/ChapWormGameplay/Items/ChapWorm/Abilities/SuperSprint/Camera/PlayerCameraModifier_ChapWorm_SuperSprint.PlayerCameraModifier_ChapWorm_SuperSprint_C"))
              (Variables|Transient|SetCameraModifier _returnvalue))
            (:else)))))))

(fn RemoveCameraMode
  ;[0.009367 0.09587 0.1042] RemoveCameraMode: Remove the applied camera mode. Use cached camera manager in case pawn has been destroyed
  (bind _cachedcameramanager (Variables|Transient|GetCachedCameraManager)
    (:then
      (Camera|RemoveCameraModeOverride _cachedcameramanager (Variables|Camera|GetCameraModeTag) 0 true)
      (bind _cameramodifier (Variables|Transient|GetCameraModifier)
        (:then
          (CameraModifier|WindDownandRemove _cameramodifier :!class /ChapWormGameplay/Items/ChapWorm/Abilities/SuperSprint/Camera/PlayerCameraModifier_ChapWorm_SuperSprint.PlayerCameraModifier_ChapWorm_SuperSprint_C)
          (Variables|Transient|SetCameraModifier 0))
        (:else)))
    (:else)))
