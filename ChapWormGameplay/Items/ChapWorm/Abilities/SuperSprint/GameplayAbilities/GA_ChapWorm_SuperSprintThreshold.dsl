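; Gameplay ability for the ChapWorm SuperSprint threshold that applies camera effects to the avatar's player while active.
; On activation it calls the parent, then adds a configured camera mode override and a threshold camera modifier via the Fort player camera manager.
; On end it removes the override and winds down the modifier before calling the parent.
; The camera manager and modifier are cached so cleanup still works if the pawn is destroyed.
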(event Ability|EventActivateAbility
  ;[0.01975 0.1997 0.0652 0.5] ActivateAbility: Apply camera mode.
  (|CallParentFunction)
  (Default|AddCameraMode))

(event Ability|EventOnEndAbility
  (params (bWasCancelled :type Boolean))
  ;[0.1441 0.06837 0.01051 0.5] EndAblity: Clean up camera.
  (Utilities|FlowControl|Sequence
    (:then_0
      (Default|RemoveCameraMode))
    (:then_1
      (|CallParentFunction bWasCancelled))))

(fn AddCameraMode
  ;[0.000237 0.04228 0.05729] AddCameraMode: Add the desired camera mode to the player pawn.
  (bind _outfortplayerpawn (CallFunction|GetAvatarasValidFortPlayerPawn)
    (:ValidFortPlayerPawn
      (Utilities|IsValid (Fortnite|GetFortPlayerControllerfromActor _outfortplayerpawn)
        (:Is%20Valid
          (bind _playercameramanager (Variables|/Script/Engine.PlayerController|GetPlayerCameraManager (Fortnite|GetFortPlayerControllerfromActor _outfortplayerpawn))
            (:then
              (bind _asfort_player_camera_base (Utilities|Casting|CastToFortPlayerCameraBase _playercameramanager))
              (bind _output_get (Variables|Transient|SetCachedCameraManager _asfort_player_camera_base)) ; Cache in case pawn is destroyed \nbefore we try to remove modifier
              (Camera|AddCameraModeOverride _output_get (Variables|Settings|GetCameraModeClass) (Variables|Settings|GetCameraModeTag) 0 0 0)
              (bind _returnvalue (Game|Player|AddNewCameraModifier _output_get "/ChapWormGameplay/Items/ChapWorm/Abilities/SuperSprint/Camera/PlayerCameraModifier_ChapWorm_SuperSprintThreshold.PlayerCameraModifier_ChapWorm_SuperSprintThreshold_C"))
              (Variables|Transient|SetCameraModifier _returnvalue))
            (:else)))))))

(fn RemoveCameraMode
  ;[0.009367 0.09587 0.1042] RemoveCameraMode: Remove the applied camera mode. Use cached camera manager in case pawn has been destroyed
  (bind _cachedcameramanager (Variables|Transient|GetCachedCameraManager)
    (:then
      (Camera|RemoveCameraModeOverride _cachedcameramanager (Variables|Settings|GetCameraModeTag) 0 true)
      (bind _cameramodifier (Variables|Transient|GetCameraModifier)
        (:then
          (CameraModifier|WindDownandRemove _cameramodifier :!class /ChapWormGameplay/Items/ChapWorm/Abilities/SuperSprint/Camera/PlayerCameraModifier_ChapWorm_SuperSprint.PlayerCameraModifier_ChapWorm_SuperSprint_C)
          (Variables|Transient|SetCameraModifier 0))
        (:else)))
    (:else)))
